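# Idle Niagara effect for the SpatialTravel pickaxe, emitting sparse rotating cubes with mesh and sprite renderers while the weapon is not swinging.
# Socket-based spawning uses two spawn groups and ten partitions to place particles along two mirrored curved profiles.
# Cubes drift with curl noise, follow source movement with fading influence, scale in and out, and carry a fading blue GlowColor.
# User.bIsSwinging stops spawning and kills live particles.
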NiagaraSystem /BRCosmetics/Weapons/FORT_Melee/Pickaxe_SpatialTravel/FX/NS_Pickaxe_SpatialTravel_Pickaxe_Idle
dependencies {
 usedDataInterfaces [&NiagaraDataInterfaceCurve, &NiagaraDataInterfaceColorCurve, &NiagaraDataInterfaceSocketReader]
 usedDynamicInputs []
 usedModules [
  &self:SystemSpawnScript.SystemScriptSource.SystemScriptGraph.NiagaraNodeAssignment_1.SetVariables_#1
  &SystemState
 ]
}
emitters {Cubes {
  emitterUpdateScript [
   &EmitterState()
   &SpawnRate(
    Spawn Probability=0.66
    SpawnRate=&MakeCustomFloatFromBool(Bool=$User.bIsSwinging, True Float=0, False Float=5)
    Use Spawn Probability=true
   )
  ]
  particleSpawnScript [
   &FN_SetSpawnGroup(Number of Spawn Groups=2)
   &PartitionParticles(Number Of Partitions=10, Partition Mode:ENiagara_PartitionMode_v2=::NewEnumerator5)
   &InitializeParticle(
    Color=<rgba 0.046 0.8162 1.5 2>
    Lifetime Max=0.8
    Lifetime Min=0.45
    Lifetime Mode=::NewEnumerator1
    Mesh Scale Mode=::NewEnumerator1
    Mesh Uniform Scale Max=1.1
    Mesh Uniform Scale Min=0.7
    Sprite Size Mode=::NewEnumerator3
    UsePositionOffset=false
    Write Material Random=true
   )
   &SocketLocation(
    SamplePosOffset=true
    Sampled Position Offset=&MakeVector(
     X=&Lerp_Float(
      A=0
      B=45
      Alpha=&FloatFromCurve(
       FloatCurve=DataInterface() {propertyValues {Curve {defaultValue=0, keys [{tangentMode=RCTM_User, time=0, value=1}, {tangentMode=RCTM_User, time=1, value=0}]}}}
       CurveIndex=$Particles.PartitionParticles.NormalizedPartition
      )
     )
     Y=&RandomRangeFloat(Minimum=-3, Maximum=3)
     Z=&MaskFloatBySpawnGroup(
      Masked Float=&FloatFromCurve(
       FloatCurve=DataInterface() {propertyValues {Curve {
          defaultValue=0
          keys [<at 0 1 RCIM_Cubic>, {arriveTangent=-0.9812, interpMode=RCIM_Cubic, leaveTangent=-0.9812, tangentMode=RCTM_User, time=1, value=0}]
         }}}
       CurveIndex=$Particles.PartitionParticles.NormalizedPartition
       Scale Curve=15
      )
      Passthrough Float=&FloatFromCurve(
       FloatCurve=DataInterface() {propertyValues {Curve {
          defaultValue=0
          keys [<at 0 1 RCIM_Cubic>, {arriveTangent=-0.9838, interpMode=RCIM_Cubic, leaveTangent=-0.9838, tangentMode=RCTM_User, time=1, value=0}]
         }}}
       CurveIndex=$Particles.PartitionParticles.NormalizedPartition
       Scale Curve=-15
      )
     )
    )
    Sampled Position Offset Coordinate Space:ENiagaraSkelMeshTransforms=::NewEnumerator2
    Socket Reader=$System.Socket Reader
    Socket Sampling Mode=::NewEnumerator2
    VelocitySpeedLimit=false
   )
  ]
  particleUpdateScript [
   &CurlNoiseForce(
    Noise Frequency=12
    Noise Quality / Cost=::NewEnumerator0
    Noise Strength=&RandomRangeFloat(Minimum=25, Maximum=100)
    Pan Noise=false
   )
   &GravityForce(Gravity=<0 0 10>)
   &Drag(Drag=1.5, Use Linear Drag=true, Use Rotational Drag=true)
   &InheritSourceMovement(
    Applied Position Scale=&Multiply_VectorByFloat(
     Vector=<1 1 1>
     Float=&FloatFromCurve(FloatCurve=DataInterface() {propertyValues {Curve {keys [<at 0 0.8>,<at 1 0>]}}})
    )
    ApplyForce=false
    ApplyPosition=true
    ApplyPositionFalloff=false
   )
   &SolveForcesAndVelocity(Clamp Velocity=false, Limit Acceleration=false)
   &ParticleState()
   &ScaleMeshSize(Scale Factor=&VectorFromFloat(Value=&FloatFromCurve(FloatCurve=DataInterface() {propertyValues {Curve {keys [
          {interpMode=RCIM_Cubic, leaveTangent=3.858, tangentMode=RCTM_Break, time=0, value=0}
          {interpMode=RCIM_Cubic, tangentMode=RCTM_Break, time=0.5, value=1.05}
          {arriveTangent=-4.3191, tangentMode=RCTM_User, time=1, value=0}
         ]}}})))
   &UpdateMeshOrientation(Rotation Rate=&RandomRangeFloat(Minimum=0.15, Maximum=0.3), Rotation Vector=<1 1 1>)
   &ScaleColor(
    Scale Alpha=&FloatFromCurve(FloatCurve=DataInterface() {propertyValues {Curve {keys [{interpMode=RCIM_Cubic, tangentMode=RCTM_Break, time=0.3, value=1}, <at 1 0>]}}})
    Scale RGB=<0 0 0>
    ScaleA=true
    ScaleRGB=true
   )
   /SetVariables(Particles.GlowColor=&ColorFromCurve(ColorCurve=DataInterface() {propertyValues {
       AlphaCurve {keys [<at 0.0011 0>,<at 0.4528 0.4 RCIM_Cubic>,<at 1 0 RCIM_Cubic>]}
       BlueCurve {keys [<at 0 2 RCIM_Cubic>,<at 1 0.4167 RCIM_Cubic>]}
       GreenCurve {keys [<at 0 0.8564 RCIM_Cubic>,<at 1 0.0401 RCIM_Cubic>]}
       RedCurve {keys [<at 0 0 RCIM_Cubic>,<at 1 0.0561 RCIM_Cubic>]}
      }})) {isSetParameters=true}
   &KillParticles(Kill Particles=$User.bIsSwinging)
  ]
  rendererClasses [&NiagaraMeshRendererProperties, &NiagaraSpriteRendererProperties]
  renderers [
   {rendererClass=&NiagaraMeshRendererProperties, rendererIndex=0}
   {rendererClass=&NiagaraSpriteRendererProperties, rendererIndex=1}
  ]
 }}
summary {
 systemName=NS_Pickaxe_SpatialTravel_Pickaxe_Idle
 emitters [{emitterName=Cubes, rendererClasses [&NiagaraMeshRendererProperties, &NiagaraSpriteRendererProperties]}]
 userVariables [User.bIsSwinging {type=NiagaraBool}]
}
